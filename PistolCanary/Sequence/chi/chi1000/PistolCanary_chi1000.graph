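# PistolCanary_chi1000 is a Level Sequence that acts as the container for chi1000, likely one cinematic shot in the chi sequence folder.
# It has no bindings of its own and is built only from subsequence tracks, split by discipline: animation, audio, camera, gameplay, lighting, techart and vfx.
# Each track holds one section over the same offset frame range, so all layers play together at a 60 fps display rate.
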LevelSequence /PistolCanary/Sequence/chi/chi1000/PistolCanary_chi1000
bindings []
display_rate {denominator=1, numerator=60}
tick_resolution {denominator=1, numerator=24000}
tracks [
 {type=MovieSceneSubTrack, display_name=animation, sections [{type=MovieSceneSubSection, end_frame=4265, start_frame=3600}]}
 {type=MovieSceneSubTrack, display_name=audio, sections [{type=MovieSceneSubSection, end_frame=4265, start_frame=3600}]}
 {type=MovieSceneSubTrack, display_name=camera, sections [{type=MovieSceneSubSection, end_frame=4265, start_frame=3600}]}
 {type=MovieSceneSubTrack, display_name=gameplay, sections [{type=MovieSceneSubSection, end_frame=4265, start_frame=3600}]}
 {type=MovieSceneSubTrack, display_name=lighting, sections [{type=MovieSceneSubSection, end_frame=4265, start_frame=3600}]}
 {type=MovieSceneSubTrack, display_name=techart, sections [{type=MovieSceneSubSection, end_frame=4265, start_frame=3600}]}
 {type=MovieSceneSubTrack, display_name=vfx, sections [{type=MovieSceneSubSection, end_frame=4265, start_frame=3600}]}
]
