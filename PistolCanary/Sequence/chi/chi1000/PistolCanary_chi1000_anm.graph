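# PistolCanary chi1000 animation Level Sequence, staging a cinematic shot with six superhero NPCs and a tentacle boss.
# The heroes hold fixed transforms, play skeletal animations and are visible only during the shot window, with overlapping second clips on Lightrider, Ziggy and MorganMyst.
# TentacleBoss combines two animation tracks with a keyed move, hold and descent.
# A no-fight barrier switches visible mid-shot, while a separate single-tentacle actor has an empty animation track.
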LevelSequence /PistolCanary/Sequence/chi/chi1000/PistolCanary_chi1000_anm
bindings [
 Lightrider {
  id=#1
  possessed_class=PistolCanary_CinematicNPC_Superhero6_C
  tracks [
   {
    type=MovieScene3DTransformTrack
    sections [{
      type=MovieScene3DTransformSection
      channels [
       Location.X {type=MovieSceneScriptingDoubleChannel, default=12200, const [12200,1]}
       Location.Y {type=MovieSceneScriptingDoubleChannel, default=79500, const [79500,1]}
       Location.Z {type=MovieSceneScriptingDoubleChannel, default=-5.325, const [3066.675,1]}
       Rotation.X {type=MovieSceneScriptingDoubleChannel, default=0, const [0,1]}
       Rotation.Y {type=MovieSceneScriptingDoubleChannel, default=0, const [0,1]}
       Rotation.Z {type=MovieSceneScriptingDoubleChannel, default=220, const [220,1]}
       Scale.X {type=MovieSceneScriptingDoubleChannel, default=1, const [1,1]}
       Scale.Y {type=MovieSceneScriptingDoubleChannel, default=1, const [1,1]}
       Scale.Z {type=MovieSceneScriptingDoubleChannel, default=1, const [1,1]}
      ]
     }]
   }
   {
    type=MovieSceneSkeletalAnimationTrack
    sections [
     {type=MovieSceneSkeletalAnimationSection, end_frame=4261, start_frame=3943}
     {type=MovieSceneSkeletalAnimationSection, end_frame=4265, start_frame=4123}
    ]
   }
   {
    type=MovieSceneVisibilityTrack
    sections [{
      type=MovieSceneVisibilitySection
      channels [None {type=MovieSceneScriptingBoolChannel, default=false, keys [[3600,false],[3943,true],[4265,false]]}]
     }]
   }
   {type=MovieSceneSpawnTrack, sections [{type=MovieSceneSpawnSection, channels [None {type=MovieSceneScriptingBoolChannel, default=true}]}]}
  ]
 }
 Synthwave {
  id=#2
  possessed_class=PistolCanary_CinematicNPC_Superhero2_C
  tracks [
   {
    type=MovieScene3DTransformTrack
    sections [{
      type=MovieScene3DTransformSection
      channels [
       Location.X {type=MovieSceneScriptingDoubleChannel, default=12600, const [12600,1]}
       Location.Y {type=MovieSceneScriptingDoubleChannel, default=2356.5931, const [79574.4062,1]}
       Location.Z {type=MovieSceneScriptingDoubleChannel, default=-78.2882, const [3068.2769,1]}
       Rotation.X {type=MovieSceneScriptingDoubleChannel, default=0, const [0,1]}
       Rotation.Y {type=MovieSceneScriptingDoubleChannel, default=0, const [0,1]}
       Rotation.Z {type=MovieSceneScriptingDoubleChannel, default=-195, const [-195,1]}
       Scale.X {type=MovieSceneScriptingDoubleChannel, default=1, const [1,1]}
       Scale.Y {type=MovieSceneScriptingDoubleChannel, default=1, const [1,1]}
       Scale.Z {type=MovieSceneScriptingDoubleChannel, default=1, const [1,1]}
      ]
     }]
   }
   {type=MovieSceneSkeletalAnimationTrack, sections [{type=MovieSceneSkeletalAnimationSection, end_frame=4265, start_frame=3943}]}
   {
    type=MovieSceneVisibilityTrack
    sections [{
      type=MovieSceneVisibilitySection
      channels [None {type=MovieSceneScriptingBoolChannel, default=false, keys [[3600,false],[3943,true],[4265,false]]}]
     }]
   }
   {type=MovieSceneSpawnTrack, sections [{type=MovieSceneSpawnSection, channels [None {type=MovieSceneScriptingBoolChannel, default=true}]}]}
  ]
 }
 PistolCanary_SingleTentacle_LAS {
  id=#3
  possessed_class=FortSkeletalMeshActor
  tracks [
   {type=MovieSceneSpawnTrack, sections [{type=MovieSceneSpawnSection, channels [None {type=MovieSceneScriptingBoolChannel, default=true}]}]}
   {
    type=MovieScene3DTransformTrack
    sections [{
      type=MovieScene3DTransformSection
      channels [
       Location.X {type=MovieSceneScriptingDoubleChannel, default=0}
       Location.Y {type=MovieSceneScriptingDoubleChannel, default=0}
       Location.Z {type=MovieSceneScriptingDoubleChannel, default=0}
       Rotation.X {type=MovieSceneScriptingDoubleChannel, default=0}
       Rotation.Y {type=MovieSceneScriptingDoubleChannel, default=0}
       Rotation.Z {type=MovieSceneScriptingDoubleChannel, default=0}
       Scale.X {type=MovieSceneScriptingDoubleChannel, default=1}
       Scale.Y {type=MovieSceneScriptingDoubleChannel, default=1}
       Scale.Z {type=MovieSceneScriptingDoubleChannel, default=1}
      ]
     }]
   }
   {type=MovieSceneSkeletalAnimationTrack, sections []}
  ]
 }
 TentacleBoss {
  id=#4
  possessed_class=BGA_Tentacle_C
  tracks [
   {type=MovieSceneSkeletalAnimationTrack, sections [{type=MovieSceneSkeletalAnimationSection, end_frame=4265, start_frame=3943}]}
   {
    type=MovieScene3DTransformTrack
    sections [{
      type=MovieScene3DTransformSection
      channels [
       Location.X {type=MovieSceneScriptingDoubleChannel, default=684, keys [[3600,683.9999],[3943,-612],[4154,-612],[4265,123.5116]]}
       Location.Y {type=MovieSceneScriptingDoubleChannel, default=-6933.6663, keys [[3600,-6933.6665],[3943,-6968],[4154,-6968],[4265,-6719.1338]]}
       Location.Z {type=MovieSceneScriptingDoubleChannel, default=448, keys [[3600,448],[3943,2688],[4154,2688],[4265,900]]}
       Rotation.X {type=MovieSceneScriptingDoubleChannel, default=0, const [0,4]}
       Rotation.Y {type=MovieSceneScriptingDoubleChannel, default=0, const [0,4]}
       Rotation.Z {type=MovieSceneScriptingDoubleChannel, default=0, const [0,4]}
       Scale.X {type=MovieSceneScriptingDoubleChannel, default=0.75, const [0.75,1]}
       Scale.Y {type=MovieSceneScriptingDoubleChannel, default=0.75, const [0.75,1]}
       Scale.Z {type=MovieSceneScriptingDoubleChannel, default=0.75, const [0.75,1]}
      ]
     }]
   }
   {type=MovieSceneSkeletalAnimationTrack, sections [{type=MovieSceneSkeletalAnimationSection, end_frame=4265, start_frame=4098}]}
  ]
 }
 Killswitch {
  id=#5
  possessed_class=PistolCanary_CinematicNPC_Superhero4_C
  tracks [
   {
    type=MovieScene3DTransformTrack
    sections [{
      type=MovieScene3DTransformSection
      channels [
       Location.X {type=MovieSceneScriptingDoubleChannel, default=12450, const [12450,1]}
       Location.Y {type=MovieSceneScriptingDoubleChannel, default=2291.2635, const [79501.5859,1]}
       Location.Z {type=MovieSceneScriptingDoubleChannel, default=-78.5693, const [3068.0029,1]}
       Rotation.X {type=MovieSceneScriptingDoubleChannel, default=0, const [0,1]}
       Rotation.Y {type=MovieSceneScriptingDoubleChannel, default=0, const [0,1]}
       Rotation.Z {type=MovieSceneScriptingDoubleChannel, default=-190, const [-190,1]}
       Scale.X {type=MovieSceneScriptingDoubleChannel, default=1, const [1,1]}
       Scale.Y {type=MovieSceneScriptingDoubleChannel, default=1, const [1,1]}
       Scale.Z {type=MovieSceneScriptingDoubleChannel, default=1, const [1,1]}
      ]
     }]
   }
   {type=MovieSceneSkeletalAnimationTrack, sections [{type=MovieSceneSkeletalAnimationSection, end_frame=4265, start_frame=3943}]}
   {
    type=MovieSceneVisibilityTrack
    sections [{
      type=MovieSceneVisibilitySection
      channels [None {type=MovieSceneScriptingBoolChannel, default=false, keys [[3600,false],[3943,true],[4265,false]]}]
     }]
   }
   {type=MovieSceneSpawnTrack, sections [{type=MovieSceneSpawnSection, channels [None {type=MovieSceneScriptingBoolChannel, default=true}]}]}
  ]
 }
 HayleeSkye {
  id=#6
  possessed_class=PistolCanary_CinematicNPC_Superhero5_C
  tracks [
   {
    type=MovieScene3DTransformTrack
    sections [{
      type=MovieScene3DTransformSection
      channels [
       Location.X {type=MovieSceneScriptingDoubleChannel, default=12520, const [12520,1]}
       Location.Y {type=MovieSceneScriptingDoubleChannel, default=1838.554, const [79525.6719,1]}
       Location.Z {type=MovieSceneScriptingDoubleChannel, default=-67.1311, const [3064.7988,1]}
       Rotation.X {type=MovieSceneScriptingDoubleChannel, default=0, const [0,1]}
       Rotation.Y {type=MovieSceneScriptingDoubleChannel, default=0, const [0,1]}
       Rotation.Z {type=MovieSceneScriptingDoubleChannel, default=-180, const [-180,1]}
       Scale.X {type=MovieSceneScriptingDoubleChannel, default=1, const [1,1]}
       Scale.Y {type=MovieSceneScriptingDoubleChannel, default=1, const [1,1]}
       Scale.Z {type=MovieSceneScriptingDoubleChannel, default=1, const [1,1]}
      ]
     }]
   }
   {type=MovieSceneSkeletalAnimationTrack, sections [{type=MovieSceneSkeletalAnimationSection, end_frame=4265, start_frame=3943}]}
   {
    type=MovieSceneVisibilityTrack
    sections [{
      type=MovieSceneVisibilitySection
      channels [None {type=MovieSceneScriptingBoolChannel, default=false, keys [[3600,false],[3943,true],[4265,false]]}]
     }]
   }
   {type=MovieSceneSpawnTrack, sections [{type=MovieSceneSpawnSection, channels [None {type=MovieSceneScriptingBoolChannel, default=true}]}]}
  ]
 }
 BP_PistolCanary_NoFightBarrier {
  id=#7
  possessed_class=BP_PistolCanary_NoFightBarrier_C
  tracks [{
    type=MovieSceneVisibilityTrack
    sections [{type=MovieSceneVisibilitySection, channels [None {type=MovieSceneScriptingBoolChannel, default=true, keys [[0,false],[4024,true]]}]}]
   }]
 }
 Ziggy {
  id=#8
  possessed_class=PistolCanary_CinematicNPC_Superhero1_C
  tracks [
   {
    type=MovieScene3DTransformTrack
    sections [{
      type=MovieScene3DTransformSection
      channels [
       Location.X {type=MovieSceneScriptingDoubleChannel, default=12262, const [12262,1]}
       Location.Y {type=MovieSceneScriptingDoubleChannel, default=79470, const [79470,1]}
       Location.Z {type=MovieSceneScriptingDoubleChannel, default=-81.7349, const [3066.1399,1]}
       Rotation.X {type=MovieSceneScriptingDoubleChannel, default=0, const [0,1]}
       Rotation.Y {type=MovieSceneScriptingDoubleChannel, default=0, const [0,1]}
       Rotation.Z {type=MovieSceneScriptingDoubleChannel, default=170, const [170,1]}
       Scale.X {type=MovieSceneScriptingDoubleChannel, default=1, const [1,1]}
       Scale.Y {type=MovieSceneScriptingDoubleChannel, default=1, const [1,1]}
       Scale.Z {type=MovieSceneScriptingDoubleChannel, default=1, const [1,1]}
      ]
     }]
   }
   {
    type=MovieSceneSkeletalAnimationTrack
    sections [
     {type=MovieSceneSkeletalAnimationSection, end_frame=4261, start_frame=3943}
     {type=MovieSceneSkeletalAnimationSection, end_frame=4265, start_frame=4161}
    ]
   }
   {
    type=MovieSceneVisibilityTrack
    sections [{
      type=MovieSceneVisibilitySection
      channels [None {type=MovieSceneScriptingBoolChannel, default=false, keys [[3600,false],[3943,true],[4265,false]]}]
     }]
   }
   {type=MovieSceneSpawnTrack, sections [{type=MovieSceneSpawnSection, channels [None {type=MovieSceneScriptingBoolChannel, default=true}]}]}
  ]
 }
 MorganMyst {
  id=#9
  possessed_class=PistolCanary_CinematicNPC_Superhero3_C
  tracks [
   {
    type=MovieScene3DTransformTrack
    sections [{
      type=MovieScene3DTransformSection
      channels [
       Location.X {type=MovieSceneScriptingDoubleChannel, default=12352, const [12356,1]}
       Location.Y {type=MovieSceneScriptingDoubleChannel, default=2740.1476, const [79458.3438,1]}
       Location.Z {type=MovieSceneScriptingDoubleChannel, default=-14.4159, const [3067.1755,1]}
       Rotation.X {type=MovieSceneScriptingDoubleChannel, default=0, const [0,1]}
       Rotation.Y {type=MovieSceneScriptingDoubleChannel, default=0, const [0,1]}
       Rotation.Z {type=MovieSceneScriptingDoubleChannel, default=-161.6094, const [-161.6094,1]}
       Scale.X {type=MovieSceneScriptingDoubleChannel, default=1, const [1,1]}
       Scale.Y {type=MovieSceneScriptingDoubleChannel, default=1, const [1,1]}
       Scale.Z {type=MovieSceneScriptingDoubleChannel, default=1, const [1,1]}
      ]
     }]
   }
   {
    type=MovieSceneSkeletalAnimationTrack
    sections [
     {type=MovieSceneSkeletalAnimationSection, end_frame=4261, start_frame=3943}
     {type=MovieSceneSkeletalAnimationSection, end_frame=4265, start_frame=4161}
    ]
   }
   {
    type=MovieSceneVisibilityTrack
    sections [{
      type=MovieSceneVisibilitySection
      channels [None {type=MovieSceneScriptingBoolChannel, default=false, keys [[3600,false],[3943,true],[4265,false]]}]
     }]
   }
   {type=MovieSceneSpawnTrack, sections [{type=MovieSceneSpawnSection, channels [None {type=MovieSceneScriptingBoolChannel, default=true}]}]}
  ]
 }
]
display_rate {denominator=1, numerator=60}
tick_resolution {denominator=1, numerator=24000}
tracks []
