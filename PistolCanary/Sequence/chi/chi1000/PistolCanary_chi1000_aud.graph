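# Audio Level Sequence for the chi1000 sequence in PistolCanary, which fires timed events rather than playing sound clips itself.
# Event trigger tracks target a possessed TentacleBoss actor with one event and the PistolCanary_AudioManager with two later events.
# Its root audio track has no sections, so any sound depends on what those events call, which is not visible here.
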LevelSequence /PistolCanary/Sequence/chi/chi1000/PistolCanary_chi1000_aud
bindings [
 TentacleBoss {
  id=#1
  possessed_class=BGA_Tentacle_C
  tracks [{
    type=MovieSceneEventTrack
    sections [{
      type=MovieSceneEventTriggerSection
      channels [None {type=MovieSceneScriptingEventChannel, const ["<Struct 'MovieSceneEvent' (0x852f46a8700) {}>", 1]}]
     }]
   }]
 }
 PistolCanary_AudioManager {
  id=#2
  possessed_class=PistolCanary_AudioManager_C
  tracks [{
    type=MovieSceneEventTrack
    sections [{
      type=MovieSceneEventTriggerSection
      channels [None {
        type=MovieSceneScriptingEventChannel
        keys [[3601, "<Struct 'MovieSceneEvent' (0x852f46a82a0) {}>"], [3901.285, "<Struct 'MovieSceneEvent' (0x852f46a8840) {}>"]]
       }]
     }]
   }]
 }
]
display_rate {denominator=1, numerator=60}
tick_resolution {denominator=1, numerator=24000}
tracks [{type=MovieSceneAudioTrack, display_name=Audio, sections []}]
