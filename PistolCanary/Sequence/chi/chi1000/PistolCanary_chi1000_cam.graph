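# Camera Level Sequence for shot chi1000 of PistolCanary, with one camera cut section spanning the shot.
# CineCamera CHI_1000 makes a keyed move in position and pitch on a wide lens, with post-process Amount and Max switching near the end.
# A spawned cameraForce Blueprint receives small camera shake and force feedback assets, with event triggers for Radius and Lobby variants of each.
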LevelSequence /PistolCanary/Sequence/chi/chi1000/PistolCanary_chi1000_cam
bindings [
 CHI_1000 {
  id=#1
  possessed_class=CineCameraActor
  tracks [{
    type=MovieScene3DTransformTrack
    sections [{
      type=MovieScene3DTransformSection
      channels [
       Location.X {type=MovieSceneScriptingDoubleChannel, default=12347, const [12353.1816,1]}
       Location.Y {type=MovieSceneScriptingDoubleChannel, default=79762, keys [[4013,79759.3438],[4281,79834.75]]}
       Location.Z {type=MovieSceneScriptingDoubleChannel, default=3100, keys [[4013,3116.0889],[4281,3080.6621]]}
       Rotation.X {type=MovieSceneScriptingDoubleChannel, default=0, const [0,1]}
       Rotation.Y {type=MovieSceneScriptingDoubleChannel, default=30, keys [[4013,26],[4281,30]]}
       Rotation.Z {type=MovieSceneScriptingDoubleChannel, default=-80, const [-79.8,1]}
       Scale.X {type=MovieSceneScriptingDoubleChannel, default=1, const [1,1]}
       Scale.Y {type=MovieSceneScriptingDoubleChannel, default=1, const [1,1]}
       Scale.Z {type=MovieSceneScriptingDoubleChannel, default=1, const [1,1]}
      ]
     }]
   }]
 }
 PistolCanary_cameraForce_BP {
  id=#2
  possessed_class=Events_Shared_cameraForce_BP_C
  tracks [
   {
    type=MovieSceneObjectPropertyTrack
    display_name=Force Effect
    sections [{
      type=MovieSceneObjectPropertySection
      channels [None {
        type=MovieSceneScriptingObjectPathChannel
        default="<Object '&Events_Shared_medium_FF' (0x8509a2598b0) Class 'ForceFeedbackEffect'>"
        const ["<Object '&Events_Shared_small_FF' (0x84c3c703480) Class 'ForceFeedbackEffect'>", 1]
       }]
     }]
   }
   {
    type=MovieSceneObjectPropertyTrack
    display_name=Cam Shake Base
    sections [{
      type=MovieSceneObjectPropertySection
      channels [None {
        type=MovieSceneScriptingObjectPathChannel
        default="<Object '&Events_Shared_00small_csb_C' (0x84c3c6a8a00) Class 'BlueprintGeneratedClass'>"
        const [
         "<Object '&Events_Shared_00small_csb_C' (0x84c3c6a8a00) Class 'BlueprintGeneratedClass'>"
         1
        ]
       }]
     }]
   }
   {
    type=MovieSceneEventTrack
    display_name=Cam Shake Radius Event Trigger
    sections [{
      type=MovieSceneEventTriggerSection
      channels [None {type=MovieSceneScriptingEventChannel, const ["<Struct 'MovieSceneEvent' (0x852f46a8700) {}>", 1]}]
     }]
   }
   {type=MovieSceneSpawnTrack, sections [{type=MovieSceneSpawnSection, channels [None {type=MovieSceneScriptingBoolChannel, default=true}]}]}
   {
    type=MovieSceneEventTrack
    display_name=Force Feedback Radius Event Trigger
    sections [{
      type=MovieSceneEventTriggerSection
      channels [None {type=MovieSceneScriptingEventChannel, const ["<Struct 'MovieSceneEvent' (0x852f46a82a0) {}>", 1]}]
     }]
   }
   {
    type=MovieScene3DTransformTrack
    sections [{
      type=MovieScene3DTransformSection
      channels [
       Location.X {type=MovieSceneScriptingDoubleChannel, default=7084}
       Location.Y {type=MovieSceneScriptingDoubleChannel, default=1572}
       Location.Z {type=MovieSceneScriptingDoubleChannel, default=0}
       Rotation.X {type=MovieSceneScriptingDoubleChannel, default=0}
       Rotation.Y {type=MovieSceneScriptingDoubleChannel, default=0}
       Rotation.Z {type=MovieSceneScriptingDoubleChannel, default=0}
       Scale.X {type=MovieSceneScriptingDoubleChannel, default=1}
       Scale.Y {type=MovieSceneScriptingDoubleChannel, default=1}
       Scale.Z {type=MovieSceneScriptingDoubleChannel, default=1}
      ]
     }]
   }
   {
    type=MovieSceneIntegerTrack
    sections [{type=MovieSceneIntegerSection, channels [None {type=MovieSceneScriptingIntegerChannel, default=2, const [0,1]}]}]
   }
   {
    type=MovieSceneEventTrack
    display_name=Cam Shake Lobby Event Trigger
    sections [{
      type=MovieSceneEventTriggerSection
      channels [None {type=MovieSceneScriptingEventChannel, const ["<Struct 'MovieSceneEvent' (0x852f46a9100) {}>", 1]}]
     }]
   }
   {
    type=MovieSceneEventTrack
    display_name=Force Feedback Lobby Event Trigger
    sections [{
      type=MovieSceneEventTriggerSection
      channels [None {type=MovieSceneScriptingEventChannel, const ["<Struct 'MovieSceneEvent' (0x852f46a9560) {}>", 1]}]
     }]
   }
  ]
 }
 CameraComponent {
  id=#3
  parent_id=#1
  possessed_class=CineCameraComponent
  tracks [
   {
    type=MovieSceneFloatTrack
    display_name=Current Focal Length
    sections [{type=MovieSceneFloatSection, channels [None {type=MovieSceneScriptingFloatChannel, default=12}]}]
   }
   {
    type=MovieSceneFloatTrack
    display_name="Manual Focus Distance (Focus Settings)"
    sections [{type=MovieSceneFloatSection, channels [None {type=MovieSceneScriptingFloatChannel, default=100000}]}]
   }
   {
    type=MovieSceneFloatTrack
    display_name=Current Aperture
    sections [{type=MovieSceneFloatSection, channels [None {type=MovieSceneScriptingFloatChannel, default=2.8}]}]
   }
   {
    type=MovieSceneFloatTrack
    display_name="Amount (Post Process Settings)"
    sections [{type=MovieSceneFloatSection, channels [None {type=MovieSceneScriptingFloatChannel, default=1, keys [[4261,0.5],[4262,1]]}]}]
   }
   {
    type=MovieSceneFloatTrack
    display_name="Max (Post Process Settings)"
    sections [{type=MovieSceneFloatSection, channels [None {type=MovieSceneScriptingFloatChannel, default=100, keys [[4261,5],[4262,100]]}]}]
   }
  ]
 }
]
display_rate {denominator=1, numerator=60}
tick_resolution {denominator=1, numerator=24000}
tracks [{type=MovieSceneCameraCutTrack, sections [{type=MovieSceneCameraCutSection, end_frame=4302, start_frame=4013}]}]
