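# Temporary camera Level Sequence for shot chi1000 of the PistolCanary event, with a camera cut and an audio track.
# The CHI_1000 cine camera drifts and tilts slightly, then lunges across the scene in a few frames as post-process Amount and Max values jump, likely motion blur.
# A spawned camera-force Blueprint carries small shake and force-feedback assets with radius and lobby event tracks, while a root event track fires twice around the fast move.
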LevelSequence /PistolCanary/Sequence/chi/chi1000/PistolCanary_chi1000_cam_temp
bindings [
 CHI_1000 {
  id=#1
  possessed_class=CineCameraActor
  tracks [{
    type=MovieScene3DTransformTrack
    sections [{
      type=MovieScene3DTransformSection
      channels [
       Location.X {type=MovieSceneScriptingDoubleChannel, default=12318.0495, keys [[4125,12321.6289],[4305,12309.0234],[4317,10760.2959]]}
       Location.Y {type=MovieSceneScriptingDoubleChannel, default=79762, keys [[4125,79762],[4305,79812],[4317,79385.5156]]}
       Location.Z {type=MovieSceneScriptingDoubleChannel, default=3036.5187, keys [[4125,3107.2314],[4305,3093.3928],[4317,3097.7388]]}
       Rotation.X {type=MovieSceneScriptingDoubleChannel, default=0, const [0,1]}
       Rotation.Y {type=MovieSceneScriptingDoubleChannel, default=28, keys [[4125,24],[4305,28]]}
       Rotation.Z {type=MovieSceneScriptingDoubleChannel, default=-75.9276, const [-75.9276,1]}
       Scale.X {type=MovieSceneScriptingDoubleChannel, default=1, const [1,1]}
       Scale.Y {type=MovieSceneScriptingDoubleChannel, default=1, const [1,1]}
       Scale.Z {type=MovieSceneScriptingDoubleChannel, default=1, const [1,1]}
      ]
     }]
   }]
 }
 PistolCanary_cameraForce_BP {
  id=#2
  possessed_class=Events_Shared_cameraForce_BP_C
  tracks [
   {
    type=MovieSceneObjectPropertyTrack
    display_name=Force Effect
    sections [{
      type=MovieSceneObjectPropertySection
      channels [None {
        type=MovieSceneScriptingObjectPathChannel
        default="<Object '&Events_Shared_medium_FF' (0x8509a2598b0) Class 'ForceFeedbackEffect'>"
        const ["<Object '&Events_Shared_small_FF' (0x84c3c703480) Class 'ForceFeedbackEffect'>", 1]
       }]
     }]
   }
   {
    type=MovieSceneObjectPropertyTrack
    display_name=Cam Shake Base
    sections [{
      type=MovieSceneObjectPropertySection
      channels [None {
        type=MovieSceneScriptingObjectPathChannel
        default="<Object '&Events_Shared_00small_csb_C' (0x84c3c6a8a00) Class 'BlueprintGeneratedClass'>"
        const [
         "<Object '&Events_Shared_00small_csb_C' (0x84c3c6a8a00) Class 'BlueprintGeneratedClass'>"
         1
        ]
       }]
     }]
   }
   {
    type=MovieSceneEventTrack
    display_name=Cam Shake Radius Event Trigger
    sections [{
      type=MovieSceneEventTriggerSection
      channels [None {type=MovieSceneScriptingEventChannel, const ["<Struct 'MovieSceneEvent' (0x852f46a96a0) {}>", 1]}]
     }]
   }
   {type=MovieSceneSpawnTrack, sections [{type=MovieSceneSpawnSection, channels [None {type=MovieSceneScriptingBoolChannel, default=true}]}]}
   {
    type=MovieSceneEventTrack
    display_name=Force Feedback Radius Event Trigger
    sections [{
      type=MovieSceneEventTriggerSection
      channels [None {type=MovieSceneScriptingEventChannel, const ["<Struct 'MovieSceneEvent' (0x852f46a9060) {}>", 1]}]
     }]
   }
   {
    type=MovieScene3DTransformTrack
    sections [{
      type=MovieScene3DTransformSection
      channels [
       Location.X {type=MovieSceneScriptingDoubleChannel, default=7084}
       Location.Y {type=MovieSceneScriptingDoubleChannel, default=1572}
       Location.Z {type=MovieSceneScriptingDoubleChannel, default=0}
       Rotation.X {type=MovieSceneScriptingDoubleChannel, default=0}
       Rotation.Y {type=MovieSceneScriptingDoubleChannel, default=0}
       Rotation.Z {type=MovieSceneScriptingDoubleChannel, default=0}
       Scale.X {type=MovieSceneScriptingDoubleChannel, default=1}
       Scale.Y {type=MovieSceneScriptingDoubleChannel, default=1}
       Scale.Z {type=MovieSceneScriptingDoubleChannel, default=1}
      ]
     }]
   }
   {
    type=MovieSceneIntegerTrack
    sections [{type=MovieSceneIntegerSection, channels [None {type=MovieSceneScriptingIntegerChannel, default=2, const [0,1]}]}]
   }
   {
    type=MovieSceneEventTrack
    display_name=Cam Shake Lobby Event Trigger
    sections [{
      type=MovieSceneEventTriggerSection
      channels [None {type=MovieSceneScriptingEventChannel, const ["<Struct 'MovieSceneEvent' (0x852f46a8f20) {}>", 1]}]
     }]
   }
   {
    type=MovieSceneEventTrack
    display_name=Force Feedback Lobby Event Trigger
    sections [{
      type=MovieSceneEventTriggerSection
      channels [None {type=MovieSceneScriptingEventChannel, const ["<Struct 'MovieSceneEvent' (0x852f46a9100) {}>", 1]}]
     }]
   }
  ]
 }
 CameraComponent {
  id=#3
  parent_id=#1
  possessed_class=CineCameraComponent
  tracks [
   {
    type=MovieSceneFloatTrack
    display_name=Current Focal Length
    sections [{type=MovieSceneFloatSection, channels [None {type=MovieSceneScriptingFloatChannel, default=12}]}]
   }
   {
    type=MovieSceneFloatTrack
    display_name="Manual Focus Distance (Focus Settings)"
    sections [{type=MovieSceneFloatSection, channels [None {type=MovieSceneScriptingFloatChannel, default=100000}]}]
   }
   {
    type=MovieSceneFloatTrack
    display_name=Current Aperture
    sections [{type=MovieSceneFloatSection, channels [None {type=MovieSceneScriptingFloatChannel, default=2.8}]}]
   }
   {
    type=MovieSceneFloatTrack
    display_name="Amount (Post Process Settings)"
    sections [{type=MovieSceneFloatSection, channels [None {type=MovieSceneScriptingFloatChannel, default=0.5, keys [[4305,0.5],[4306,1]]}]}]
   }
   {
    type=MovieSceneFloatTrack
    display_name="Max (Post Process Settings)"
    sections [{type=MovieSceneFloatSection, channels [None {type=MovieSceneScriptingFloatChannel, default=5, keys [[4305,5],[4306,100]]}]}]
   }
  ]
 }
]
display_rate {denominator=1, numerator=60}
tick_resolution {denominator=1, numerator=24000}
tracks [
 {
  type=MovieSceneAudioTrack
  display_name=Audio
  sections [{
    type=MovieSceneAudioSection
    end_frame=4725
    start_frame=3547
    channels [
     Volume {type=MovieSceneScriptingFloatChannel, default=1}
     Pitch {type=MovieSceneScriptingFloatChannel, default=1}
     AttachActor {type=MovieSceneScriptingActorReferenceChannel, default="<Struct 'MovieSceneObjectBindingID' (0x852f46be7f0) {}>"}
    ]
   }]
 }
 {
  type=MovieSceneEventTrack
  sections [{
    type=MovieSceneEventTriggerSection
    channels [None {
      type=MovieSceneScriptingEventChannel
      keys [[4275, "<Struct 'MovieSceneEvent' (0x852f46a9a60) {}>"], [4316, "<Struct 'MovieSceneEvent' (0x852f46a9240) {}>"]]
     }]
   }]
 }
 {type=MovieSceneCameraCutTrack, sections [{type=MovieSceneCameraCutSection, end_frame=4697, start_frame=3599}]}
]
