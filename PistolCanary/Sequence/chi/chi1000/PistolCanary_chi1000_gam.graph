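# Gameplay event Level Sequence for PistolCanary's chi1000 sequence, containing only event trigger tracks and no animation.
# It possesses a focus point, four NPC aim actors and a MakeEnvironmentIndestructible actor, firing events on each.
# The focus point and indestructibility actor each get two timed events, while each aim actor and an unbound root track carry one.
# The functions those events call are not shown here.
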LevelSequence /PistolCanary/Sequence/chi/chi1000/PistolCanary_chi1000_gam
bindings [
 BP_PistolCanary_FocusPoint {
  id=#1
  possessed_class=BP_PistolCanary_FocusPoint_C
  tracks [{
    type=MovieSceneEventTrack
    sections [{
      type=MovieSceneEventTriggerSection
      channels [None {
        type=MovieSceneScriptingEventChannel
        keys [[3601, "<Struct 'MovieSceneEvent' (0x852f46a9a60) {}>"], [4169, "<Struct 'MovieSceneEvent' (0x852f46a8f20) {}>"]]
       }]
     }]
   }]
 }
 BGA_PistolCanary_NPCAimActor4 {
  id=#2
  possessed_class=BGA_PistolCanary_NPCAimActor_C
  tracks [{
    type=MovieSceneEventTrack
    sections [{
      type=MovieSceneEventTriggerSection
      channels [None {type=MovieSceneScriptingEventChannel, const ["<Struct 'MovieSceneEvent' (0x852f46a8fc0) {}>", 1]}]
     }]
   }]
 }
 BGA_PistolCanary_NPCAimActor {
  id=#3
  possessed_class=BGA_PistolCanary_NPCAimActor_C
  tracks [{
    type=MovieSceneEventTrack
    sections [{
      type=MovieSceneEventTriggerSection
      channels [None {type=MovieSceneScriptingEventChannel, const ["<Struct 'MovieSceneEvent' (0x852f46a96a0) {}>", 1]}]
     }]
   }]
 }
 BGA_PistolCanary_NPCAimActor2 {
  id=#4
  possessed_class=BGA_PistolCanary_NPCAimActor_C
  tracks [{
    type=MovieSceneEventTrack
    sections [{
      type=MovieSceneEventTriggerSection
      channels [None {type=MovieSceneScriptingEventChannel, const ["<Struct 'MovieSceneEvent' (0x852f46a8de0) {}>", 1]}]
     }]
   }]
 }
 BP_PistolCanary_MakeEnvironmentIndestructible {
  id=#5
  possessed_class=BP_PistolCanary_MakeEnvironmentIndestructible_C
  tracks [{
    type=MovieSceneEventTrack
    sections [{
      type=MovieSceneEventTriggerSection
      channels [None {
        type=MovieSceneScriptingEventChannel
        keys [[4199, "<Struct 'MovieSceneEvent' (0x852f46a9920) {}>"], [4226, "<Struct 'MovieSceneEvent' (0x852f46a94c0) {}>"]]
       }]
     }]
   }]
 }
 BGA_PistolCanary_NPCAimActor3 {
  id=#6
  possessed_class=BGA_PistolCanary_NPCAimActor_C
  tracks [{
    type=MovieSceneEventTrack
    sections [{
      type=MovieSceneEventTriggerSection
      channels [None {type=MovieSceneScriptingEventChannel, const ["<Struct 'MovieSceneEvent' (0x852f46a9420) {}>", 1]}]
     }]
   }]
 }
]
display_rate {denominator=1, numerator=60}
tick_resolution {denominator=1, numerator=24000}
tracks [{
  type=MovieSceneEventTrack
  sections [{
    type=MovieSceneEventTriggerSection
    channels [None {type=MovieSceneScriptingEventChannel, const ["<Struct 'MovieSceneEvent' (0x852f46a9600) {}>", 1]}]
   }]
 }]
