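# Lighting Level Sequence for shot chi1000 of PistolCanary, adding enhancement spotlights named for a comic intro.
# Four positioned spotlights share one keyed spawn window, so they exist only during that span.
# Two low-intensity cool-blue lights contrast with much stronger orange and magenta lights, some with tuned cone or radius.
# A DaySequenceModifierVolume is also positioned, with its component's User Blend Weight driven to apply the day-sequence override.
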LevelSequence /PistolCanary/Sequence/chi/chi1000/PistolCanary_chi1000_lgt
bindings [
 LightComponent0 {
  id=#1
  parent_id=#2
  possessed_class=SpotLightComponent
  tracks [
   {
    type=MovieSceneFloatTrack
    display_name=Intensity
    sections [{type=MovieSceneFloatSection, channels [None {type=MovieSceneScriptingFloatChannel, default=15, const [15,1]}]}]
   }
   {
    type=MovieSceneColorTrack
    display_name=Light Color
    sections [{
      type=MovieSceneColorSection
      channels [
       Color.R {type=MovieSceneScriptingFloatChannel, default=0.6}
       Color.G {type=MovieSceneScriptingFloatChannel, default=0.7}
       Color.B {type=MovieSceneScriptingFloatChannel, default=1}
       Color.A {type=MovieSceneScriptingFloatChannel, default=1}
      ]
     }]
   }
  ]
 }
 L_ComicIntro-Enhance_01 {
  id=#3
  possessed_class=SpotLight
  tracks [
   {
    type=MovieSceneSpawnTrack
    sections [{
      type=MovieSceneSpawnSection
      channels [None {type=MovieSceneScriptingBoolChannel, default=false, keys [[3600,false],[3940,true],[4322,false]]}]
     }]
   }
   {
    type=MovieScene3DTransformTrack
    sections [{
      type=MovieScene3DTransformSection
      channels [
       Location.X {type=MovieSceneScriptingDoubleChannel, default=12031.0855}
       Location.Y {type=MovieSceneScriptingDoubleChannel, default=79572}
       Location.Z {type=MovieSceneScriptingDoubleChannel, default=3389.9916}
       Rotation.X {type=MovieSceneScriptingDoubleChannel, default=0}
       Rotation.Y {type=MovieSceneScriptingDoubleChannel, default=-40}
       Rotation.Z {type=MovieSceneScriptingDoubleChannel, default=-10}
       Scale.X {type=MovieSceneScriptingDoubleChannel, default=1}
       Scale.Y {type=MovieSceneScriptingDoubleChannel, default=1}
       Scale.Z {type=MovieSceneScriptingDoubleChannel, default=1}
      ]
     }]
   }
  ]
 }
 DaySequenceModifier {
  id=#4
  parent_id=#5
  possessed_class=DaySequenceModifierComponent
  tracks [{
    type=MovieSceneFloatTrack
    display_name=User Blend Weight
    sections [{type=MovieSceneFloatSection, channels [None {type=MovieSceneScriptingFloatChannel, default=0.5641, const [1,1]}]}]
   }]
 }
 L_ComicIntro-Enhance_5 {
  id=#6
  possessed_class=SpotLight
  tracks [
   {
    type=MovieSceneSpawnTrack
    sections [{
      type=MovieSceneSpawnSection
      channels [None {type=MovieSceneScriptingBoolChannel, default=false, keys [[3600,false],[3940,true],[4322,false]]}]
     }]
   }
   {
    type=MovieScene3DTransformTrack
    sections [{
      type=MovieScene3DTransformSection
      channels [
       Location.X {type=MovieSceneScriptingDoubleChannel, default=12895.0859}
       Location.Y {type=MovieSceneScriptingDoubleChannel, default=79812}
       Location.Z {type=MovieSceneScriptingDoubleChannel, default=3233.9917}
       Rotation.X {type=MovieSceneScriptingDoubleChannel, default=18.7897}
       Rotation.Y {type=MovieSceneScriptingDoubleChannel, default=-3.7825}
       Rotation.Z {type=MovieSceneScriptingDoubleChannel, default=-157.4271}
       Scale.X {type=MovieSceneScriptingDoubleChannel, default=1}
       Scale.Y {type=MovieSceneScriptingDoubleChannel, default=1}
       Scale.Z {type=MovieSceneScriptingDoubleChannel, default=1}
      ]
     }]
   }
  ]
 }
 L_ComicIntro-Enhance_4 {
  id=#7
  possessed_class=SpotLight
  tracks [
   {
    type=MovieSceneSpawnTrack
    sections [{
      type=MovieSceneSpawnSection
      channels [None {type=MovieSceneScriptingBoolChannel, default=false, keys [[3600,false],[3940,true],[4322,false]]}]
     }]
   }
   {
    type=MovieScene3DTransformTrack
    sections [{
      type=MovieScene3DTransformSection
      channels [
       Location.X {type=MovieSceneScriptingDoubleChannel, default=12807.437}
       Location.Y {type=MovieSceneScriptingDoubleChannel, default=79369.7142}
       Location.Z {type=MovieSceneScriptingDoubleChannel, default=3298.1323}
       Rotation.X {type=MovieSceneScriptingDoubleChannel, default=-41.9301}
       Rotation.Y {type=MovieSceneScriptingDoubleChannel, default=-6.8785}
       Rotation.Z {type=MovieSceneScriptingDoubleChannel, default=148.0699}
       Scale.X {type=MovieSceneScriptingDoubleChannel, default=1}
       Scale.Y {type=MovieSceneScriptingDoubleChannel, default=1}
       Scale.Z {type=MovieSceneScriptingDoubleChannel, default=1}
      ]
     }]
   }
  ]
 }
 LightComponent0 {
  id=#8
  parent_id=#3
  possessed_class=SpotLightComponent
  tracks [
   {
    type=MovieSceneFloatTrack
    display_name=Intensity
    sections [{type=MovieSceneFloatSection, channels [None {type=MovieSceneScriptingFloatChannel, default=30, const [30,1]}]}]
   }
   {
    type=MovieSceneColorTrack
    display_name=Light Color
    sections [{
      type=MovieSceneColorSection
      channels [
       Color.R {type=MovieSceneScriptingFloatChannel, default=0.6}
       Color.G {type=MovieSceneScriptingFloatChannel, default=0.7}
       Color.B {type=MovieSceneScriptingFloatChannel, default=1}
       Color.A {type=MovieSceneScriptingFloatChannel, default=1}
      ]
     }]
   }
  ]
 }
 LightComponent0 {
  id=#9
  parent_id=#6
  possessed_class=SpotLightComponent
  tracks [
   {
    type=MovieSceneFloatTrack
    display_name=Intensity
    sections [{type=MovieSceneFloatSection, channels [None {type=MovieSceneScriptingFloatChannel, default=250, const [250,1]}]}]
   }
   {
    type=MovieSceneColorTrack
    display_name=Light Color
    sections [{
      type=MovieSceneColorSection
      channels [
       Color.R {type=MovieSceneScriptingFloatChannel, default=1}
       Color.G {type=MovieSceneScriptingFloatChannel, default=0.6}
       Color.B {type=MovieSceneScriptingFloatChannel, default=0.1}
       Color.A {type=MovieSceneScriptingFloatChannel, default=1}
      ]
     }]
   }
   {
    type=MovieSceneFloatTrack
    display_name=Attenuation Radius
    sections [{type=MovieSceneFloatSection, channels [None {type=MovieSceneScriptingFloatChannel, default=2000}]}]
   }
   {
    type=MovieSceneFloatTrack
    display_name=Outer Cone Angle
    sections [{type=MovieSceneFloatSection, channels [None {type=MovieSceneScriptingFloatChannel, default=30}]}]
   }
  ]
 }
 L_ComicIntro-Enhance_3 {
  id=#2
  possessed_class=SpotLight
  tracks [
   {
    type=MovieSceneSpawnTrack
    sections [{
      type=MovieSceneSpawnSection
      channels [None {type=MovieSceneScriptingBoolChannel, default=false, keys [[3600,false],[3940,true],[4322,false]]}]
     }]
   }
   {
    type=MovieScene3DTransformTrack
    sections [{
      type=MovieScene3DTransformSection
      channels [
       Location.X {type=MovieSceneScriptingDoubleChannel, default=12123.0855}
       Location.Y {type=MovieSceneScriptingDoubleChannel, default=79904}
       Location.Z {type=MovieSceneScriptingDoubleChannel, default=3101.9916}
       Rotation.X {type=MovieSceneScriptingDoubleChannel, default=-10.7821}
       Rotation.Y {type=MovieSceneScriptingDoubleChannel, default=29.2791}
       Rotation.Z {type=MovieSceneScriptingDoubleChannel, default=-58.7721}
       Scale.X {type=MovieSceneScriptingDoubleChannel, default=1}
       Scale.Y {type=MovieSceneScriptingDoubleChannel, default=1}
       Scale.Z {type=MovieSceneScriptingDoubleChannel, default=1}
      ]
     }]
   }
  ]
 }
 LightComponent0 {
  id=#10
  parent_id=#7
  possessed_class=SpotLightComponent
  tracks [
   {
    type=MovieSceneFloatTrack
    display_name=Intensity
    sections [{type=MovieSceneFloatSection, channels [None {type=MovieSceneScriptingFloatChannel, default=200, const [200,1]}]}]
   }
   {
    type=MovieSceneColorTrack
    display_name=Light Color
    sections [{
      type=MovieSceneColorSection
      channels [
       Color.R {type=MovieSceneScriptingFloatChannel, default=1}
       Color.G {type=MovieSceneScriptingFloatChannel, default=0}
       Color.B {type=MovieSceneScriptingFloatChannel, default=0.5}
       Color.A {type=MovieSceneScriptingFloatChannel, default=1}
      ]
     }]
   }
   {
    type=MovieSceneFloatTrack
    display_name=Outer Cone Angle
    sections [{type=MovieSceneFloatSection, channels [None {type=MovieSceneScriptingFloatChannel, default=20}]}]
   }
  ]
 }
 DaySequenceModifierVolume_PistolCanary {
  id=#5
  possessed_class=DaySequenceModifierVolume
  tracks [{
    type=MovieScene3DTransformTrack
    sections [{
      type=MovieScene3DTransformSection
      channels [
       Location.X {type=MovieSceneScriptingDoubleChannel, default=14456}
       Location.Y {type=MovieSceneScriptingDoubleChannel, default=77488}
       Location.Z {type=MovieSceneScriptingDoubleChannel, default=3688}
       Rotation.X {type=MovieSceneScriptingDoubleChannel, default=0}
       Rotation.Y {type=MovieSceneScriptingDoubleChannel, default=0}
       Rotation.Z {type=MovieSceneScriptingDoubleChannel, default=0}
       Scale.X {type=MovieSceneScriptingDoubleChannel, default=1}
       Scale.Y {type=MovieSceneScriptingDoubleChannel, default=1}
       Scale.Z {type=MovieSceneScriptingDoubleChannel, default=1}
      ]
     }]
   }]
 }
]
display_rate {denominator=1, numerator=60}
tick_resolution {denominator=1, numerator=24000}
tracks []
