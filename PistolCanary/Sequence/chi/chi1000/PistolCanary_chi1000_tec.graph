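# Technical Level Sequence for PistolCanary shot chi1000, driving comic-style post-processing and panel events, apparently for a hero intro.
# The PistolCanary_ComicEffect volume's blend weight falls from full to zero.
# Meanwhile the PC_Comic_Intro_PP actor, held at a fixed, enlarged transform, toggles visibility while its child PostProcess component fades in, out, then in again later.
# Event tracks on PistolCanary_PanelManager trigger Dev Widget, timed Hero Intro and Comic Border events.
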LevelSequence /PistolCanary/Sequence/chi/chi1000/PistolCanary_chi1000_tec
bindings [
 PistolCanary_ComicEffect {
  id=#1
  possessed_class=PostProcessVolume
  tracks [{
    type=MovieSceneFloatTrack
    display_name=Blend Weight
    sections [{type=MovieSceneFloatSection, channels [None {type=MovieSceneScriptingFloatChannel, default=0, keys [[3600,1],[4231,0]]}]}]
   }]
 }
 PostProcess {
  id=#2
  parent_id=#3
  possessed_class=PostProcessComponent
  tracks [{
    type=MovieSceneFloatTrack
    display_name=Blend Weight
    sections [{
      type=MovieSceneFloatSection
      channels [None {type=MovieSceneScriptingFloatChannel, default=1, keys [[3600,0],[4013,1],[4100,1],[4189,0],[7786,0],[7875,1]]}]
     }]
   }]
 }
 PC_Comic_Intro_PP {
  id=#3
  possessed_class=PC_Comic_Intro_PP_C
  tracks [
   {
    type=MovieSceneVisibilityTrack
    sections [{
      type=MovieSceneVisibilitySection
      channels [None {type=MovieSceneScriptingBoolChannel, default=false, keys [[3600,true],[4231,false],[7786,true]]}]
     }]
   }
   {
    type=MovieScene3DTransformTrack
    sections [{
      type=MovieScene3DTransformSection
      channels [
       Location.X {type=MovieSceneScriptingDoubleChannel, default=14488}
       Location.Y {type=MovieSceneScriptingDoubleChannel, default=72664}
       Location.Z {type=MovieSceneScriptingDoubleChannel, default=3724}
       Rotation.X {type=MovieSceneScriptingDoubleChannel, default=0}
       Rotation.Y {type=MovieSceneScriptingDoubleChannel, default=0}
       Rotation.Z {type=MovieSceneScriptingDoubleChannel, default=0}
       Scale.X {type=MovieSceneScriptingDoubleChannel, default=350}
       Scale.Y {type=MovieSceneScriptingDoubleChannel, default=350}
       Scale.Z {type=MovieSceneScriptingDoubleChannel, default=350}
      ]
     }]
   }
  ]
 }
 PistolCanary_PanelManager {
  id=#4
  possessed_class=PistolCanary_PanelManager_C
  tracks [
   {
    type=MovieSceneEventTrack
    display_name=Dev Widget
    sections [{
      type=MovieSceneEventTriggerSection
      channels [None {type=MovieSceneScriptingEventChannel, const ["<Struct 'MovieSceneEvent' (0x852f46aa8c0) {}>", 1]}]
     }]
   }
   {
    type=MovieSceneEventTrack
    display_name=Hero Intro
    sections [{
      type=MovieSceneEventTriggerSection
      channels [None {
        type=MovieSceneScriptingEventChannel
        keys [
         [3864, "<Struct 'MovieSceneEvent' (0x852f46aa000) {}>"]
         [4064, "<Struct 'MovieSceneEvent' (0x852f46a9e20) {}>"]
         [4264, "<Struct 'MovieSceneEvent' (0x852f46aa0a0) {}>"]
         [4300, "<Struct 'MovieSceneEvent' (0x852f46aa280) {}>"]
        ]
       }]
     }]
   }
   {
    type=MovieSceneEventTrack
    display_name=Comic Border
    sections [{
      type=MovieSceneEventTriggerSection
      channels [None {type=MovieSceneScriptingEventChannel, const ["<Struct 'MovieSceneEvent' (0x852f46aa460) {}>", 1]}]
     }]
   }
  ]
 }
]
display_rate {denominator=1, numerator=60}
tick_resolution {denominator=1, numerator=24000}
tracks []
