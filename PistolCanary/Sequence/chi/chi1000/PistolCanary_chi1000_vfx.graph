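# VFX Level Sequence for shot chi1000 of PistolCanary, driving two possessed Niagara actors: a roar and a tentacle ground portal.
# NS_Roar is held at a fixed, heavily scaled-up transform, with Auto Activate and FX System keys triggering then deactivating it.
# The portal uses a Niagara system life-cycle section spanning the roar's window, with constant transforms on both actor and component and an empty FX System track.
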LevelSequence /PistolCanary/Sequence/chi/chi1000/PistolCanary_chi1000_vfx
bindings [
 NS_Roar {
  id=#1
  possessed_class=NiagaraActor
  tracks [{
    type=MovieScene3DTransformTrack
    sections [{
      type=MovieScene3DTransformSection
      channels [
       Location.X {type=MovieSceneScriptingDoubleChannel, default=16242.3516, const [16242.3516,1]}
       Location.Y {type=MovieSceneScriptingDoubleChannel, default=71420, const [71420,1]}
       Location.Z {type=MovieSceneScriptingDoubleChannel, default=4787.9668, const [4787.9668,1]}
       Rotation.X {type=MovieSceneScriptingDoubleChannel, default=20, const [20,1]}
       Rotation.Y {type=MovieSceneScriptingDoubleChannel, default=0, const [0,1]}
       Rotation.Z {type=MovieSceneScriptingDoubleChannel, default=90, const [90,1]}
       Scale.X {type=MovieSceneScriptingDoubleChannel, default=25, const [25,1]}
       Scale.Y {type=MovieSceneScriptingDoubleChannel, default=25, const [25,1]}
       Scale.Z {type=MovieSceneScriptingDoubleChannel, default=25, const [25,1]}
      ]
     }]
   }]
 }
 NiagaraComponent0 {
  id=#2
  parent_id=#1
  possessed_class=NiagaraComponent
  tracks [
   {
    type=MovieSceneBoolTrack
    display_name=Auto Activate
    sections [{
      type=MovieSceneBoolSection
      channels [None {type=MovieSceneScriptingBoolChannel, default=true, keys [[3600,false],[3850,true],[4217,false]]}]
     }]
   }
   {
    type=MovieSceneParticleTrack
    display_name=FX System
    sections [{
      type=MovieSceneParticleSection
      channels [None {type=MovieSceneScriptingParticleChannel, keys [[3850, "<ParticleKey.TRIGGER: 2>"], [4217, "<ParticleKey.DEACTIVATE: 1>"]]}]
     }]
   }
  ]
 }
 NS_PistolCanary_Tentacle_GroundPortal_01 {
  id=#3
  possessed_class=NiagaraActor
  tracks [{
    type=MovieScene3DTransformTrack
    sections [{
      type=MovieScene3DTransformSection
      channels [
       Location.X {type=MovieSceneScriptingDoubleChannel, default=14348, const [14348,1]}
       Location.Y {type=MovieSceneScriptingDoubleChannel, default=69801, const [69801,1]}
       Location.Z {type=MovieSceneScriptingDoubleChannel, default=4356, const [4356,1]}
       Rotation.X {type=MovieSceneScriptingDoubleChannel, default=0, const [0,1]}
       Rotation.Y {type=MovieSceneScriptingDoubleChannel, default=0, const [0,1]}
       Rotation.Z {type=MovieSceneScriptingDoubleChannel, default=0, const [0,1]}
       Scale.X {type=MovieSceneScriptingDoubleChannel, default=1, const [1,1]}
       Scale.Y {type=MovieSceneScriptingDoubleChannel, default=1.1, const [1.1,1]}
       Scale.Z {type=MovieSceneScriptingDoubleChannel, default=1.1, const [1.1,1]}
      ]
     }]
   }]
 }
 NiagaraComponent0 {
  id=#4
  parent_id=#3
  possessed_class=NiagaraComponent
  tracks [
   {
    type=MovieSceneParticleTrack
    display_name=FX System
    sections [{type=MovieSceneParticleSection, channels [None {type=MovieSceneScriptingParticleChannel}]}]
   }
   {type=MovieSceneNiagaraSystemTrack, sections [{type=MovieSceneNiagaraSystemSpawnSection, end_frame=4331, start_frame=3572}]}
   {
    type=MovieScene3DTransformTrack
    sections [{
      type=MovieScene3DTransformSection
      channels [
       Location.X {type=MovieSceneScriptingDoubleChannel, default=14328, const [14328,1]}
       Location.Y {type=MovieSceneScriptingDoubleChannel, default=70128, const [70128,1]}
       Location.Z {type=MovieSceneScriptingDoubleChannel, default=4328, const [4328,1]}
       Rotation.X {type=MovieSceneScriptingDoubleChannel, default=0, const [0,1]}
       Rotation.Y {type=MovieSceneScriptingDoubleChannel, default=0, const [0,1]}
       Rotation.Z {type=MovieSceneScriptingDoubleChannel, default=0, const [0,1]}
       Scale.X {type=MovieSceneScriptingDoubleChannel, default=1.15, const [1.15,1]}
       Scale.Y {type=MovieSceneScriptingDoubleChannel, default=1.15, const [1.15,1]}
       Scale.Z {type=MovieSceneScriptingDoubleChannel, default=1.05, const [1.05,1]}
      ]
     }]
   }
  ]
 }
]
display_rate {denominator=1, numerator=60}
tick_resolution {denominator=1, numerator=24000}
tracks []
